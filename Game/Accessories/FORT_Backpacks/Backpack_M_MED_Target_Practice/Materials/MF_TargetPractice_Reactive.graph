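# Material function for the Target Practice backpack that shows a two-digit kill count and animated, kill-reactive lights.
# UV1 splits tens and units, each offset into a number flipbook atlas, colored and optionally written to base color and emissive.
# Time since a Blueprint-supplied timestamp drives a sine glow pulse and switches lights to a faster, brighter pattern.
# Vertex color channels set blink order, optional timing offsets and parts hidden while reactive.
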MaterialFunction /Game/Accessories/FORT_Backpacks/Backpack_M_MED_Target_Practice/Materials/MF_TargetPractice_Reactive
graph {nodes [
  n1 = VertexColor()
  Add(
   A=@n1
   B=TextureSample(UVs=TextureCoordinate()) {texture=/Game/Accessories/FORT_Backpacks/Backpack_M_MED_Target_Practice/Textures/T_M_MED_TargetPractice_Backpack_FX}.RGB
  )
  ComponentMask(None=VectorParameter() {group=TargetPractice, parameterName=TP_Column_RowNumber, defaultValue=<rgba 2 1 0 1>}.RGB) {g=true, r=true}
  ComponentMask(None=TextureCoordinate()) {r=true}
  n2 = Time()
  ConstantBiasScale(None=Sine(None=Multiply(A=@n2) {constB=0.25}))
  n3 = Time()
  ConstantBiasScale(None=Sine(None=Multiply(A=@n3) {constB=0.25}))
  ConstantBiasScale(None=Sine(None=Time()))
  Constant() {r=1}
  n4 = MaterialFunctionCall(Float2=TextureCoordinate()) {materialFunction=&BreakOutFloat2Components}
  n5 = ScalarParameter() {defaultValue=10, group=TargetPractice, parameterName=TP_No_Columns}
  Divide(A=Ceil(None=Multiply(A=@n4.R, B=@n5)), B=@n5)
  n6 = GetMaterialAttributes(None=FunctionInput() {inputName=Material, inputType=FunctionInput_MaterialAttributes}) {attributeGetTypes [#1, #2]}
  n7 = Round(None=ComponentMask(None=TextureCoordinate() {coordinateIndex=1}) {r=true})
  n8 = Multiply(A=TextureCoordinate() {coordinateIndex=1}, B=Constant2Vector() {g=1.5, r=2}) {constB=2}
  n9 = MaterialFunctionCall(Float2=@n8) {materialFunction=&BreakOutFloat2Components}
  n10 = Add(A=@n9.R) {constB=-1}
  n11 = MaterialFunctionCall(Float2=LinearInterpolate(
    A=@n8
    B=MaterialFunctionCall(X=@n10, Y=@n9.G) {materialFunction=&MakeFloat2}.Result
    Alpha=@n7
   )) {materialFunction=&BreakOutFloat2Components}
  n12 = Divide(A=Constant() {r=1}, B=@n5)
  n13 = ScalarParameter() {group=TargetPractice, parameterName=TP_KillsInput}
  n14 = Constant() {r=10}
  n15 = Divide(A=@n13, B=@n14) {constB=10}
  n16 = LinearInterpolate(A=Subtract(A=@n15, B=Frac(None=@n15)), B=Fmod(A=@n13, B=@n14), Alpha=@n7)
  n17 = MaterialFunctionCall(X=Add(A=Multiply(A=@n11.R, B=@n12), B=Multiply(A=@n12, B=@n16)) {constB=0}, Y=@n11.G) {materialFunction=&MakeFloat2}
  n18 = ComponentMask(None=Multiply(A=@n12, B=VectorParameter() {group=TargetPractice, parameterName=TP_Units_Offset, defaultValue=<rgba 0 -1.6 0 1>}.RGB)) {g=true, r=true}
  n19 = ComponentMask(None=Multiply(A=@n12, B=VectorParameter() {group=TargetPractice, parameterName=TP_Tens_Offset, defaultValue=<rgba 0 -1.6 0 1>}.RGB)) {g=true, r=true}
  n20 = TextureSampleParameter2D(UVs=LinearInterpolate(A=Add(A=@n17.Result, B=@n18), B=Add(A=@n17.Result, B=@n19), Alpha=@n7)) {
   group=TargetPractice
   parameterName=TP_AtlasTexture
   samplerType=SAMPLERTYPE_Masks
   texture=/Game/Global/Textures/Default/Flipbooks/T_NumberFlipBook
  }
  n21 = Round(None=ChannelMaskParameter(None=@n20.RGB) {group=TargetPractice, parameterName=TP_NumberChannel}.R)
  n22 = Multiply(
   A=LinearInterpolate(
    A=VectorParameter() {group=TargetPractice, parameterName=TP_Color_01_Color, defaultValue=<rgba 1 0.7836 0 1>}.RGB
    B=VectorParameter() {group=TargetPractice, parameterName=TP_Color_02_Color, defaultValue=<rgba 1 0 0.507 1>}.RGB
    Alpha=@n7
   )
   B=@n21
  )
  n23 = ScalarParameter() {group=TargetPractice, parameterName=TP_TimeOffset}
  n24 = LinearInterpolate(
   A=@n23
   B=Multiply(A=ChannelMaskParameter(None=@n1) {group=TargetPractice, parameterName=TP_TimingOffsetChannel}.R, B=@n23)
   Alpha=ScalarParameter() {group=TargetPractice, parameterName=TP_UseVertexColorChannelForTimeOFfset}
  )
  n25 = Subtract(A=Add(A=Time(), B=@n24), B=ScalarParameter() {defaultValue=-100000000, group=TargetPractice, parameterName=TP_GameTimeFromBP})
  n26 = Divide(A=@n25, B=ScalarParameter() {defaultValue=2, group=TargetPractice, parameterName=TP_ReactiveGlowDuration})
  n27 = Sine(None=Add(A=Saturate(None=@n26)) {constB=-0.25})
  n28 = Multiply(
   A=@n22
   B=Multiply(A=ScalarParameter() {defaultValue=10, group=TargetPractice, parameterName=TP_BaseGlowStrength}, B=ConstantBiasScale(None=@n27))
  )
  n29 = ChannelMaskParameter(None=@n1) {group=TargetPractice, maskChannel=Green, parameterName=TP_BlinkOrderChannel, defaultValue=<rgba 0 1 0 0>}
  n30 = Time()
  n31 = ScalarParameter() {defaultValue=0.5, group=TargetPractice, parameterName=TP_LightsDefaultSpeed}
  n32 = ScalarParameter() {defaultValue=2, group=TargetPractice, parameterName=TP_LightsReactiveSpeed}
  n33 = Saturate(None=Divide(A=@n25, B=ScalarParameter() {defaultValue=3, group=TargetPractice, parameterName=TP_ReactiveBlinkDuration}))
  n34 = ConstantBiasScale(None=Sine(None=Add(A=@n33) {constB=-0.25}))
  n35 = Ceil(None=@n34)
  n36 = Frac(None=Multiply(A=Add(A=@n30, B=@n24), B=LinearInterpolate(A=@n31, B=@n32, Alpha=@n35)))
  n37 = Constant() {r=0.05}
  n38 = Constant() {r=1}
  n39 = SphereMask(A=@n29.R, B=@n36, Radius=@n37, Hardness=@n38)
  n40 = Ceil(None=@n29.R)
  n41 = Add(A=Subtract(A=@n29.R, B=OneMinus(None=@n36))) {constB=0.05}
  n42 = Multiply(A=Ceil(None=Clamp(None=@n41)), B=@n40)
  n43 = ChannelMaskParameter(None=@n1) {group=TargetPractice, parameterName=TP_HideArrowChannel}
  n44 = ScalarParameter() {defaultValue=1, group=TargetPractice, parameterName=TP_ValueToHideonReactive}
  n45 = Clamp(None=OneMinus(None=SphereMask(A=@n43.R, B=@n44, Radius=@n37, Hardness=@n38)))
  n46 = Multiply(A=@n6.EmissiveColor, B=LinearInterpolate(A=Multiply(A=@n39, B=@n40), B=Multiply(A=@n42, B=@n45), Alpha=@n35))
  n47 = ScalarParameter() {defaultValue=5, group=TargetPractice, parameterName=TP_BlinkingLight_Strength}
  n48 = LinearInterpolate(
   A=@n47
   B=Multiply(A=@n47, B=ScalarParameter() {defaultValue=2, group=TargetPractice, parameterName=TP_BlinkingLight_ReactiveMultiplier})
   Alpha=@n35
  )
  n49 = Add(
   A=StaticSwitchParameter(True=Power(Base=@n28), False=Constant3Vector()) {defaultValue=true, group=TargetPractice, parameterName=TP_UseNumbers}
   B=Power(Base=Multiply(A=@n46, B=@n48))
  )
  n50 = LinearInterpolate(A=@n6.BaseColor, B=@n22, Alpha=@n21)
  FunctionOutput(None=SetMaterialAttributes!(
    MaterialAttributes=@n6.MaterialAttributes
    EmissiveColor=Add(A=@n6.EmissiveColor, B=@n49)
    BaseColor=StaticSwitchParameter(True=@n50, False=@n6.BaseColor) {defaultValue=true, group=TargetPractice, parameterName=TP_UseNumbers}
   ) {attributeSetTypes [#1, #2]})
  Multiply(A=@n29.R, B=@n29.R)
  Round(None=@n34)
  TextureSampleParameter2D() {group=TargetPractice, parameterName=TP_AtlasTexture, texture=&T_NumberFlipBook}
  TextureSampleParameter2D() {
   parameterName=TP_EmissiveTexture
   texture=/Game/Accessories/FORT_Backpacks/Backpack_M_MED_Target_Practice/Textures/T_M_MED_TargetPractice_Backpack_E
  }
 ]}
