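# Character-style skinned material for the Target Practice backpack cosmetic, built on material attributes with Default Lit shading.
# The shared character PreFX function feeds optional cloth fuzz, thin film and metal LUT stages, each toggled by a static switch.
# SkinFX_Mask drives those stages, with selectable channels for fuzz and thin film.
# The result passes through MF_TargetPractice_Reactive, named for a reactive effect, before the shared character PostFX function.
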Material /Game/Accessories/FORT_Backpacks/Backpack_M_MED_Target_Practice/Materials/M_MED_Target_Practice_Backpack
graph {
 properties {
  bUseMaterialAttributes=true
  bUsedWithMeshDeformer=true
  bUsedWithMorphTargets=true
  bUsedWithSkeletalMesh=true
  pixelDepthOffsetMode=PDOM_Legacy
  usedShadingModels=Default Lit
 }
 nodes [
  n1 = MaterialFunctionCall() {materialFunction=&BaseFN_CharacterFunction_PreFX}
  n2 = TextureSampleParameter2D() {parameterName=SkinFX_Mask}
  n3 = StaticSwitchParameter(
   True=MaterialFunctionCall!(
    Alpha=StaticComponentMaskParameter(None=@n2.RGBA) {defaultR=true, group=Skin FX, parameterName=Cloth Channel}
    Gmap=@n1.Gmap
    Material=@n1.Gmap
   ) {materialFunction=/Game/Weapons/WeaponSkins/MaterialFunction/ClothFuzz}.Result
   False=@n1.Result
  ) {group=Skin FX, parameterName=Use Cloth fuzz}
  n4 = StaticSwitchParameter(
   True=MaterialFunctionCall!(Alpha=StaticComponentMaskParameter(None=@n2.RGBA) {defaultR=true, group=Skin FX, parameterName=ThinFilm_Channel}, Material Attributes=@n3) {materialFunction=/Game/Weapons/WeaponSkins/MaterialFunction/ThinFilmFunction}.Result
   False=@n3
  ) {group=Skin FX, parameterName=Use Thin Film}
  n5 = MaterialFunctionCall(Mask=@n2.RGBA, Materials=@n4) {materialFunction=&MF_MetalsLUT}
  MaterialFunctionCall(
   AO=@n1.AO
   Distance Blend=@n1.AO
   Fresnel=@n1.AO
   Material=MaterialFunctionCall!(Material=StaticSwitchParameter(True=@n5.Result, False=@n4) {group=Skin FX, parameterName=UseMetalLUT}) {materialFunction=&/MF_TargetPractice_Reactive}.Result
  ) {materialFunction=&BaseFN_CharacterFunction_PostFX}
 ]
}
